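; Sound library component that plays Spatial Travel audio for a Fort pawn: movement, jump, land, boost, water, in-air and readiness cues.
; On BeginPlay it finds the pawn via its owner chain and adds a first- or third-person library by local view, destroying itself where cosmetics cannot run.
; Event functions play tagged one-shots or call TriggerLoopEvent, often returning the loop's audio component.
; Started loops register in a tag-to-fade-time map used for fade-outs and bulk stopping.
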(event EventBeginPlay
  (bind _1psoundlibrary (Variables|Default|Get1pSoundLibrary))
  (bind _3psoundlibrary (Variables|Default|Get3pSoundLibrary))
  (Networking|CanExecuteCosmeticEvents
    (:True
      (CallFunction|GetPlayerPawn)
      (bind _playerpawn (Variables|Default|GetPlayerPawn)
        (:then
          (Utilities|IsValid (select (Pawn|IsLocallyViewed _playerpawn) _1psoundlibrary _3psoundlibrary)
            (:Is%20Valid
              (SoundLibrary|AddLibrary (select (Pawn|IsLocallyViewed _playerpawn) _1psoundlibrary _3psoundlibrary)))))
        (:else)))
    (:False
      (Components|DestroyComponent))))

(fn CheckForLoop
  (params (Tag :type Gameplay%20Tag%20Structure%20%28by%20ref%29))
  (bind _returnvalue (SoundLibrary|MakeSoundLibraryTag Tag))
  (bind _loopingsoundlibrarytagsfadetime (Variables|Default|GetLoopingSoundLibraryTagsFadeTime))
  (if (not (Utilities|Map|Contains _loopingsoundlibrarytagsfadetime _returnvalue))
    (Utilities|Map|Add _loopingsoundlibrarytagsfadetime _returnvalue 0.5))
  (return))

(fn GetPlayerPawn
  (bind _returnvalue (Components|GetOwner))
  (bind _asfort_pawn (Utilities|Casting|CastToFortPawn _returnvalue)
    (:then
      (Variables|Default|SetPlayerPawn _asfort_pawn))
    (:CastFailed
      (bind _asfort_pawn_2 (Utilities|Casting|CastToFortPawn (Actor|GetOwner _returnvalue))
        (:then
          (Variables|Default|SetPlayerPawn _asfort_pawn_2))
        (:CastFailed
          (Utilities|String|LogString (Utilities|String|Append "SpatialTravelComponent Error: " "Component on " (Utilities|GetDisplayName _returnvalue) " was unable to find a fortpawn"))))))
  (return))

(fn GrabAudComp
  (params (SoundLibraryPlayResult :type Sound%20Library%20Play%20Result%20Structure%20%28by%20ref%29))
  (bind (_eventname _bwasblocked _bsuccess _audiocomponents) (Utilities|Struct|BreakSoundLibraryPlayResult SoundLibraryPlayResult))
  (if (Utilities|Array|IsValidIndex _audiocomponents)
    (return (Utilities|Array|Get(acopy) _audiocomponents))
    (else
      (return 0))))

(fn Logger
  (params (Message%20Type :type Integer) (Event%20Message :type String))
  (Utilities|String|LogString (Utilities|String|Append "Spatial Travel Audio Component on " (Utilities|GetDisplayName (Components|GetOwner)) (Utilities|Select :Index Message%20Type :Option%200 " triggered event: " :Option%201 " stopped event: " :Option%202 " error: ") Event%20Message)))

(fn OnAirborne
  (params (Starting :type Boolean))
  (bind _soundlibraryplayresult (CallFunction|TriggerLoopEvent "(TagName=\"SoundLibrary.SpatialTravel.InAir\")" Starting))
  (bind _audcomp (CallFunction|GrabAudComp _soundlibraryplayresult))
  (return _audcomp))

(fn OnBoostChange
  (params (IsBoosting :type Boolean))
  (if IsBoosting
    (SoundLibrary|PlayEvent "(TagName=\"SoundLibrary.SpatialTravel.Boost.Start\")")
    (bind _soundlibraryplayresult (CallFunction|TriggerLoopEvent "(TagName=\"SoundLibrary.SpatialTravel.Boost.Loop\")" true))
    (bind _audcomp (CallFunction|GrabAudComp _soundlibraryplayresult))
    (return _audcomp true)
    (else
      (SoundLibrary|PlayEvent "(TagName=\"SoundLibrary.SpatialTravel.Boost.Stop\")")
      (CallFunction|TriggerLoopEvent "(TagName=\"SoundLibrary.SpatialTravel.Boost.Loop\")")
      (return 0 false))))

(fn OnBoostReady
  (SoundLibrary|PlayEvent "(TagName=\"SoundLibrary.SpatialTravel.UI.BoostReady\")")
  (Variables|Default|GetDebug
    (:then
      (CallFunction|Logger 0 "BoostReady"))
    (:else)))

(fn OnDismount
  (CallFunction|TriggerLoopEvent "(TagName=\"SoundLibrary.SpatialTravel.InAir\")"))

(fn OnJump
  (SoundLibrary|PlayEvent "(TagName=\"SoundLibrary.SpatialTravel.Jump\")")
  (bind _soundlibraryplayresult (CallFunction|TriggerLoopEvent "(TagName=\"SoundLibrary.SpatialTravel.InAir\")" true))
  (bind _audcomp (CallFunction|GrabAudComp _soundlibraryplayresult))
  (return _audcomp))

(fn OnLand
  (SoundLibrary|PlayEvent "(TagName=\"SoundLibrary.SpatialTravel.Land\")")
  (CallFunction|TriggerLoopEvent "(TagName=\"SoundLibrary.SpatialTravel.InAir\")"))

(fn OnMovementChange
  (params (Starting :type Boolean))
  (if Starting
    (SoundLibrary|PlayEvent "(TagName=\"SoundLibrary.SpatialTravel.Start\")")
    (bind _soundlibraryplayresult (CallFunction|TriggerLoopEvent "(TagName=\"SoundLibrary.SpatialTravel.Movement\")" true))
    (bind _audcomp (CallFunction|GrabAudComp _soundlibraryplayresult))
    (return _audcomp)
    (else
      (SoundLibrary|PlayEvent "(TagName=\"SoundLibrary.SpatialTravel.Stop\")")
      (CallFunction|TriggerLoopEvent "(TagName=\"SoundLibrary.SpatialTravel.Movement\")")
      (return 0))))

(fn OnMovementReady
  (SoundLibrary|PlayEvent "(TagName=\"SoundLibrary.SpatialTravel.UI.BoostReady\")")
  (Variables|Default|GetDebug
    (:then
      (CallFunction|Logger 0 "MovementReady"))
    (:else)))

(fn OnPreBoost
  (bind _preboost (LocalVariables|GetPreboost))
  (SoundLibrary|PlayEvent _preboost)
  (Variables|Default|GetDebug
    (:then
      (CallFunction|Logger 0 (GameplayTags|GetDebugStringfromGameplayTag _preboost)))
    (:else)))

(fn OnSoundStopped
  (params (InEventName :type Gameplay%20Tag%20Structure) (InComponent :type Audio%20Component%20Object%20Reference))
  (bind (_value _returnvalue) (Utilities|Map|Find (Variables|Default|GetLoopingSoundLibraryTagsFadeTime) (SoundLibrary|MakeSoundLibraryTag InEventName)))
  (Audio|Components|Audio|FadeOut InComponent (select _returnvalue _value 0.5) 0.0 "Linear" :!class /Script/Engine.AudioComponent)
  (return true))

(fn OnWatermovement
  (params (Starting :type Boolean))
  (bind _soundlibraryplayresult (CallFunction|TriggerLoopEvent "(TagName=\"SoundLibrary.SpatialTravel.WaterMovement\")" Starting))
  (bind _audcomp (CallFunction|GrabAudComp _soundlibraryplayresult))
  (return _audcomp))

(fn StopLoopingSounds
  (bind _keys (Utilities|Map|Keys (Variables|Default|GetLoopingSoundLibraryTagsFadeTime)))
  (for _array_element _keys
    (SoundLibrary|StopEvent _array_element)))

(fn TriggerLoopEvent
  (params (InEventName :type Gameplay%20Tag%20Structure) (Starting :type Boolean))
  (bind _output_get (LocalVariables|SetEventTag InEventName))
  (LocalVariables|SetDebugString (GameplayTags|GetDebugStringfromGameplayTag _output_get))
  (if Starting
    ; Play the event
    (bind _returnvalue (SoundLibrary|PlayEvent (LocalVariables|GetEventTag)))
    (Variables|Default|GetDebug
      (:then
        (CallFunction|Logger 0 (LocalVariables|GetDebugString)))
      (:else
        _))
    (CallFunction|CheckforLoop (LocalVariables|GetEventTag))
    (return _returnvalue)
    (elif (SoundLibrary|IsEventActive (LocalVariables|GetEventTag))
      ; Stop the event, triggering a fade out
      (SoundLibrary|StopEvent (LocalVariables|GetEventTag))
      (Variables|Default|GetDebug
        (:then
          (CallFunction|Logger 1 (LocalVariables|GetDebugString)))
        (:else
          _))
      (return 0)
      (else
        ; Ignore stop if there is no active sound playing
        (Variables|Default|GetDebug
          (:then
            (CallFunction|Logger 2 (Utilities|String|Append (LocalVariables|GetDebugString) " attempted fade on non-playing sound")))
          (:else
            _))))))
